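; GoldenGateBeamModule_Graph is a PCG example graph that builds an IPN beam module for a Golden Gate architecture example.
; Three Create_Point subgraphs define the top flange, web and bottom flange of the I-profile as separate points.
; All three points merge into the primary input of one Spawn_Assets subgraph, whose result is the graph output.
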(graph GoldenGateBeamModule_Graph
  (bind TopFlange_Point (Subgraph :Graph /PCGPrimitives/Primitives/Create/Create_Point.Create_Point))
  (bind Web_Point (Subgraph :Graph /PCGPrimitives/Primitives/Create/Create_Point.Create_Point))
  (bind BottomFlange_Point (Subgraph :Graph /PCGPrimitives/Primitives/Create/Create_Point.Create_Point))
  (bind SpawnIPNBeam (Subgraph :Graph /PCGPrimitives/Primitives/Spawn/Spawn_Assets.Spawn_Assets))
  (connect BottomFlange_Point Out SpawnIPNBeam PrimaryInput)
  (connect TopFlange_Point Out SpawnIPNBeam PrimaryInput)
  (connect Web_Point Out SpawnIPNBeam PrimaryInput)
  (output Out SpawnIPNBeam))
