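; PCG example graph that builds a Golden Gate Bridge from PCGPrimitives subgraphs.
; Three created points connect into a tagged spine, which is offset into six road lanes and two sidewalks.
; Points placed along these lines, some offset again, receive copied sections or referenced-actor modules for roadway, sidewalks, fences, lampposts and under-deck structure, then spawn as assets.
; Towers copy referenced actor data onto north and south points, while four standalone points spawn platforms.
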(graph GoldenGateBridge_Graph
  (bind NorthTowerLocation (Subgraph :Graph /PCGPrimitives/Primitives/Create/Create_Point.Create_Point))
  (bind SouthTowerLocation (Subgraph :Graph /PCGPrimitives/Primitives/Create/Create_Point.Create_Point))
  ;[0.15 0.15 0.15 0.5 70] Bridge spine
  (bind CablePoint_North (Subgraph :Graph /PCGPrimitives/Primitives/Create/Create_Point.Create_Point))
  (bind CablePoint_Center (Subgraph :Graph /PCGPrimitives/Primitives/Create/Create_Point.Create_Point))
  (bind CablePoint_South (Subgraph :Graph /PCGPrimitives/Primitives/Create/Create_Point.Create_Point))
  (bind ConnectCablePoints (Subgraph :Graph /PCGPrimitives/Primitives/Trace/Trace_Connect.Trace_Connect))
  (bind Subgraph_3 (Subgraph :Graph /PCGPrimitives/Primitives/Write/Write_Tag.Write_Tag))
  ;[0.15 0.15 0.15 0.5 70] TOWERS
  (bind GetPillarReference (Subgraph :Graph /PCGPrimitives/Primitives/Get/Get_ActorDataByRef.Get_ActorDataByRef))
  (bind CopyPillarsToBothLocations (Subgraph :Graph /PCGPrimitives/Primitives/Copy/CopyToPoints.CopyToPoints))
  (bind Subgraph_64 (Subgraph :Graph /PCGPrimitives/Primitives/Spawn/Spawn_Assets.Spawn_Assets))
  (bind Subgraph_21 (Subgraph :Graph /PCGPrimitives/Primitives/Copy/CopyToPoints.CopyToPoints))
  (bind Subgraph_22 (Subgraph :Graph /PCGPrimitives/Primitives/Spawn/Spawn_Assets.Spawn_Assets))
  (bind Subgraph_0 (Subgraph :Graph /PCGPrimitives/Primitives/Get/Get_ActorDataByRef.Get_ActorDataByRef))
  (bind Subgraph_4 (Subgraph :Graph /PCGPrimitives/Primitives/Copy/CopyToPoints.CopyToPoints))
  (bind Subgraph_5 (Subgraph :Graph /PCGPrimitives/Primitives/Copy/CopyToPoints.CopyToPoints))
  (bind Subgraph_6 (Subgraph :Graph /PCGPrimitives/Primitives/Spawn/Spawn_Assets.Spawn_Assets))
  ;[0.15 0.15 0.15 0.5 70] SIDEWALK
  (bind OffsetEastSidewalk (Subgraph :Graph /PCGPrimitives/Primitives/Transform/Transform_Offset.Transform_Offset))
  (bind OffsetWestSidewalk (Subgraph :Graph /PCGPrimitives/Primitives/Transform/Transform_Offset.Transform_Offset))
  (bind SidewalkSection_Point (Subgraph :Graph /PCGPrimitives/Primitives/Create/Create_Point.Create_Point))
  (bind Subgraph_2 (Subgraph :Graph /PCGPrimitives/Primitives/Place/Place_OnSpline.Place_OnSpline))
  (bind Subgraph_1 (Subgraph :Graph /PCGPrimitives/Primitives/Copy/CopyToPoints.CopyToPoints))
  (bind Subgraph_70 (Subgraph :Graph /PCGPrimitives/Primitives/Spawn/Spawn_Assets.Spawn_Assets))
  ;[0.15 0.15 0.15 0.5 70] ROAD WAY
  (bind RoadLaneSection_Point (Subgraph :Graph /PCGPrimitives/Primitives/Create/Create_Point.Create_Point))
  (bind OffsetLane1 (Subgraph :Graph /PCGPrimitives/Primitives/Transform/Transform_Offset.Transform_Offset))
  (bind OffsetLane2 (Subgraph :Graph /PCGPrimitives/Primitives/Transform/Transform_Offset.Transform_Offset))
  (bind OffsetLane3 (Subgraph :Graph /PCGPrimitives/Primitives/Transform/Transform_Offset.Transform_Offset))
  (bind OffsetLane4 (Subgraph :Graph /PCGPrimitives/Primitives/Transform/Transform_Offset.Transform_Offset))
  (bind OffsetLane5 (Subgraph :Graph /PCGPrimitives/Primitives/Transform/Transform_Offset.Transform_Offset))
  (bind OffsetLane6 (Subgraph :Graph /PCGPrimitives/Primitives/Transform/Transform_Offset.Transform_Offset))
  (bind PlaceOnAllLanes (Subgraph :Graph /PCGPrimitives/Primitives/Place/Place_OnSpline.Place_OnSpline))
  (bind CopyRoadLaneSections (Subgraph :Graph /PCGPrimitives/Primitives/Copy/CopyToPoints.CopyToPoints))
  (bind SpawnRoadLanes (Subgraph :Graph /PCGPrimitives/Primitives/Spawn/Spawn_Assets.Spawn_Assets))
  (bind Subgraph_69 (Subgraph :Graph /PCGPrimitives/Primitives/Transform/Transform_Offset.Transform_Offset))
  (bind Subgraph_71 (Subgraph :Graph /PCGPrimitives/Primitives/Create/Create_Point.Create_Point))
  (bind Subgraph_72 (Subgraph :Graph /PCGPrimitives/Primitives/Copy/CopyToPoints.CopyToPoints))
  (bind Subgraph_73 (Subgraph :Graph /PCGPrimitives/Primitives/Spawn/Spawn_Assets.Spawn_Assets))
  (bind Subgraph_7 (Subgraph :Graph /PCGPrimitives/Primitives/Debug/Debug_PrimitiveData.Debug_PrimitiveData))
  ;[0.15 0.15 0.15 0.5 70] EXTERNAL FENCE
  (bind SampleEastSidewalkForFence (Subgraph :Graph /PCGPrimitives/Primitives/Place/Place_OnSpline.Place_OnSpline))
  (bind SampleWestSidewalkForFence (Subgraph :Graph /PCGPrimitives/Primitives/Place/Place_OnSpline.Place_OnSpline))
  (bind OffsetEastFenceOutward (Subgraph :Graph /PCGPrimitives/Primitives/Transform/Transform_Offset.Transform_Offset))
  (bind OffsetWestFenceOutward (Subgraph :Graph /PCGPrimitives/Primitives/Transform/Transform_Offset.Transform_Offset))
  (bind Subgraph_40 (Subgraph :Graph /PCGPrimitives/Primitives/Get/Get_ActorDataByRef.Get_ActorDataByRef))
  (bind CopyFencePostsToEastSide (Subgraph :Graph /PCGPrimitives/Primitives/Copy/CopyToPoints.CopyToPoints))
  (bind CopyFencePostsToWestSide (Subgraph :Graph /PCGPrimitives/Primitives/Copy/CopyToPoints.CopyToPoints))
  (bind SpawnFencePosts (Subgraph :Graph /PCGPrimitives/Primitives/Spawn/Spawn_Assets.Spawn_Assets))
  ;[0.15 0.15 0.15 0.5 70] UNDER DECK STRUCTURE
  (bind GetDeckModuleReference (Subgraph :Graph /PCGPrimitives/Primitives/Get/Get_ActorDataByRef.Get_ActorDataByRef))
  (bind LowerEastSidewalForDeck (Subgraph :Graph /PCGPrimitives/Primitives/Transform/Transform_Offset.Transform_Offset))
  (bind LowerWestSidewalkForDeck (Subgraph :Graph /PCGPrimitives/Primitives/Transform/Transform_Offset.Transform_Offset))
  (bind Subgraph_52 (Subgraph :Graph /PCGPrimitives/Primitives/Transform/Transform_Offset.Transform_Offset))
  (bind SampleEastSideEvery15m (Subgraph :Graph /PCGPrimitives/Primitives/Place/Place_OnSpline.Place_OnSpline))
  (bind Subgraph_53 (Subgraph :Graph /PCGPrimitives/Primitives/Transform/Transform_Offset.Transform_Offset))
  (bind SampleWestSideEvery15m (Subgraph :Graph /PCGPrimitives/Primitives/Place/Place_OnSpline.Place_OnSpline))
  (bind CopyDeckModulesToEastSide (Subgraph :Graph /PCGPrimitives/Primitives/Copy/CopyToPoints.CopyToPoints))
  (bind CopyDeckModulesToWestSide (Subgraph :Graph /PCGPrimitives/Primitives/Copy/CopyToPoints.CopyToPoints))
  (bind SpawnDeckModules (Subgraph :Graph /PCGPrimitives/Primitives/Spawn/Spawn_Assets.Spawn_Assets))
  ;[0.15 0.15 0.15 0.5 70] LAMPOST
  (bind GetLamppostReference (Subgraph :Graph /PCGPrimitives/Primitives/Get/Get_ActorDataByRef.Get_ActorDataByRef))
  (bind SampleEastSidewalkForLampposts (Subgraph :Graph /PCGPrimitives/Primitives/Place/Place_OnSpline.Place_OnSpline))
  (bind SampleWestSidewalkForLampposts (Subgraph :Graph /PCGPrimitives/Primitives/Place/Place_OnSpline.Place_OnSpline))
  (bind OffsetEastLamppostsInward (Subgraph :Graph /PCGPrimitives/Primitives/Transform/Transform_Offset.Transform_Offset))
  (bind OffsetWestLamppostsInward (Subgraph :Graph /PCGPrimitives/Primitives/Transform/Transform_Offset.Transform_Offset))
  (bind CopyLamppostsToEastSide (Subgraph :Graph /PCGPrimitives/Primitives/Copy/CopyToPoints.CopyToPoints))
  (bind CopyLamppostsToWestSide (Subgraph :Graph /PCGPrimitives/Primitives/Copy/CopyToPoints.CopyToPoints))
  (bind SpawnLampposts (Subgraph :Graph /PCGPrimitives/Primitives/Spawn/Spawn_Assets.Spawn_Assets))
  ;[0.15 0.15 0.15 0.5 70] PLATFORMS
  (bind NorthTower_EastPlatform_Point (Subgraph :Graph /PCGPrimitives/Primitives/Create/Create_Point.Create_Point))
  (bind NorthTower_WestPlatform_Point (Subgraph :Graph /PCGPrimitives/Primitives/Create/Create_Point.Create_Point))
  (bind SouthTower_EastPlatform_Point (Subgraph :Graph /PCGPrimitives/Primitives/Create/Create_Point.Create_Point))
  (bind SouthTower_WestPlatform_Point (Subgraph :Graph /PCGPrimitives/Primitives/Create/Create_Point.Create_Point))
  (bind SpawnPlatforms (Subgraph :Graph /PCGPrimitives/Primitives/Spawn/Spawn_Assets.Spawn_Assets))
  ;[0.15 0.15 0.15 0.5 70] INTERNAL FENCE
  (bind GetInternalFenceModuleReference (Subgraph :Graph /PCGPrimitives/Primitives/Get/Get_ActorDataByRef.Get_ActorDataByRef))
  (bind SampleEastSidewalkForInternalFence (Subgraph :Graph /PCGPrimitives/Primitives/Place/Place_OnSpline.Place_OnSpline))
  (bind SampleWestSidewalkForInternalFence (Subgraph :Graph /PCGPrimitives/Primitives/Place/Place_OnSpline.Place_OnSpline))
  (bind OffsetEastInternalFenceInward (Subgraph :Graph /PCGPrimitives/Primitives/Transform/Transform_Offset.Transform_Offset))
  (bind OffsetWestInternalFenceInward (Subgraph :Graph /PCGPrimitives/Primitives/Transform/Transform_Offset.Transform_Offset))
  (bind CopyInternalFenceToEastSide (Subgraph :Graph /PCGPrimitives/Primitives/Copy/CopyToPoints.CopyToPoints))
  (bind CopyInternalFenceToWestSide (Subgraph :Graph /PCGPrimitives/Primitives/Copy/CopyToPoints.CopyToPoints))
  (bind SpawnInternalFence (Subgraph :Graph /PCGPrimitives/Primitives/Spawn/Spawn_Assets.Spawn_Assets))
  (reroute Bridge%20Spine Subgraph_3)
  (reroute East%20Sidewalk OffsetEastSidewalk)
  (reroute North%20Tower%20Location NorthTowerLocation)
  (reroute South%20Tower%20Location SouthTowerLocation)
  (reroute West%20Sidewalk OffsetWestSidewalk)
  (connect CablePoint_Center Out ConnectCablePoints PrimaryInput)
  (connect CablePoint_North Out ConnectCablePoints PrimaryInput)
  (connect CablePoint_South Out ConnectCablePoints PrimaryInput)
  (connect GetDeckModuleReference Out CopyDeckModulesToEastSide Source)
  (connect SampleEastSideEvery15m Out CopyDeckModulesToEastSide Target)
  (connect GetDeckModuleReference Out CopyDeckModulesToWestSide Source)
  (connect SampleWestSideEvery15m Out CopyDeckModulesToWestSide Target)
  (connect Subgraph_40 Out CopyFencePostsToEastSide Source)
  (connect OffsetEastFenceOutward Out CopyFencePostsToEastSide Target)
  (connect Subgraph_40 Out CopyFencePostsToWestSide Source)
  (connect OffsetWestFenceOutward Out CopyFencePostsToWestSide Target)
  (connect GetInternalFenceModuleReference Out CopyInternalFenceToEastSide Source)
  (connect OffsetEastInternalFenceInward Out CopyInternalFenceToEastSide Target)
  (connect GetInternalFenceModuleReference Out CopyInternalFenceToWestSide Source)
  (connect OffsetWestInternalFenceInward Out CopyInternalFenceToWestSide Target)
  (connect GetLamppostReference Out CopyLamppostsToEastSide Source)
  (connect OffsetEastLamppostsInward Out CopyLamppostsToEastSide Target)
  (connect GetLamppostReference Out CopyLamppostsToWestSide Source)
  (connect OffsetWestLamppostsInward Out CopyLamppostsToWestSide Target)
  (connect GetPillarReference Out CopyPillarsToBothLocations Source)
  (connect North%20Tower%20Location Out CopyPillarsToBothLocations Target)
  (connect RoadLaneSection_Point Out CopyRoadLaneSections Source)
  (connect PlaceOnAllLanes Out CopyRoadLaneSections Target)
  (connect East%20Sidewalk Out LowerEastSidewalForDeck PrimaryInput)
  (connect West%20Sidewalk Out LowerWestSidewalkForDeck PrimaryInput)
  (connect SampleEastSidewalkForFence Out OffsetEastFenceOutward PrimaryInput)
  (connect SampleEastSidewalkForInternalFence Out OffsetEastInternalFenceInward PrimaryInput)
  (connect SampleEastSidewalkForLampposts Out OffsetEastLamppostsInward PrimaryInput)
  (connect Bridge%20Spine Out OffsetEastSidewalk PrimaryInput)
  (connect Bridge%20Spine Out OffsetLane1 PrimaryInput)
  (connect Bridge%20Spine Out OffsetLane2 PrimaryInput)
  (connect Bridge%20Spine Out OffsetLane3 PrimaryInput)
  (connect Bridge%20Spine Out OffsetLane4 PrimaryInput)
  (connect Bridge%20Spine Out OffsetLane5 PrimaryInput)
  (connect Bridge%20Spine Out OffsetLane6 PrimaryInput)
  (connect SampleWestSidewalkForFence Out OffsetWestFenceOutward PrimaryInput)
  (connect SampleWestSidewalkForInternalFence Out OffsetWestInternalFenceInward PrimaryInput)
  (connect SampleWestSidewalkForLampposts Out OffsetWestLamppostsInward PrimaryInput)
  (connect Bridge%20Spine Out OffsetWestSidewalk PrimaryInput)
  (connect OffsetLane1 Out PlaceOnAllLanes PrimaryInput)
  (connect OffsetLane2 Out PlaceOnAllLanes PrimaryInput)
  (connect OffsetLane3 Out PlaceOnAllLanes PrimaryInput)
  (connect OffsetLane4 Out PlaceOnAllLanes PrimaryInput)
  (connect OffsetLane5 Out PlaceOnAllLanes PrimaryInput)
  (connect OffsetLane6 Out PlaceOnAllLanes PrimaryInput)
  (connect LowerEastSidewalForDeck Out SampleEastSideEvery15m PrimaryInput)
  (connect Subgraph_52 Out SampleEastSideEvery15m PrimaryInput)
  (connect East%20Sidewalk Out SampleEastSidewalkForFence PrimaryInput)
  (connect East%20Sidewalk Out SampleEastSidewalkForInternalFence PrimaryInput)
  (connect East%20Sidewalk Out SampleEastSidewalkForLampposts PrimaryInput)
  (connect LowerWestSidewalkForDeck Out SampleWestSideEvery15m PrimaryInput)
  (connect Subgraph_53 Out SampleWestSideEvery15m PrimaryInput)
  (connect West%20Sidewalk Out SampleWestSidewalkForFence PrimaryInput)
  (connect West%20Sidewalk Out SampleWestSidewalkForInternalFence PrimaryInput)
  (connect West%20Sidewalk Out SampleWestSidewalkForLampposts PrimaryInput)
  (connect CopyDeckModulesToEastSide Out SpawnDeckModules PrimaryInput)
  (connect CopyDeckModulesToWestSide Out SpawnDeckModules PrimaryInput)
  (connect CopyFencePostsToEastSide Out SpawnFencePosts PrimaryInput)
  (connect CopyFencePostsToWestSide Out SpawnFencePosts PrimaryInput)
  (connect CopyInternalFenceToEastSide Out SpawnInternalFence PrimaryInput)
  (connect CopyInternalFenceToWestSide Out SpawnInternalFence PrimaryInput)
  (connect CopyLamppostsToEastSide Out SpawnLampposts PrimaryInput)
  (connect CopyLamppostsToWestSide Out SpawnLampposts PrimaryInput)
  (connect NorthTower_EastPlatform_Point Out SpawnPlatforms PrimaryInput)
  (connect NorthTower_WestPlatform_Point Out SpawnPlatforms PrimaryInput)
  (connect SouthTower_EastPlatform_Point Out SpawnPlatforms PrimaryInput)
  (connect SouthTower_WestPlatform_Point Out SpawnPlatforms PrimaryInput)
  (connect CopyRoadLaneSections Out SpawnRoadLanes PrimaryInput)
  (connect SidewalkSection_Point Out Subgraph_1 Source)
  (connect Subgraph_2 Out Subgraph_1 Target)
  (connect OffsetEastSidewalk Out Subgraph_2 PrimaryInput)
  (connect OffsetWestSidewalk Out Subgraph_2 PrimaryInput)
  (connect ConnectCablePoints Out Subgraph_3 PrimaryInput)
  (connect Subgraph_0 Out Subgraph_4 Source)
  (connect North%20Tower%20Location Out Subgraph_4 Target)
  (connect Subgraph_0 Out Subgraph_5 Source)
  (connect South%20Tower%20Location Out Subgraph_5 Target)
  (connect Subgraph_4 Out Subgraph_6 PrimaryInput)
  (connect Subgraph_5 Out Subgraph_6 PrimaryInput)
  (connect PlaceOnAllLanes Out Subgraph_7 PrimaryInput)
  (connect GetPillarReference Out Subgraph_21 Source)
  (connect South%20Tower%20Location Out Subgraph_21 Target)
  (connect Subgraph_21 Out Subgraph_22 PrimaryInput)
  (connect East%20Sidewalk Out Subgraph_52 PrimaryInput)
  (connect West%20Sidewalk Out Subgraph_53 PrimaryInput)
  (connect CopyPillarsToBothLocations Out Subgraph_64 PrimaryInput)
  (connect PlaceOnAllLanes Out Subgraph_69 PrimaryInput)
  (connect Subgraph_1 Out Subgraph_70 PrimaryInput)
  (connect Subgraph_71 Out Subgraph_72 Source)
  (connect Subgraph_69 Out Subgraph_72 Target)
  (connect Subgraph_72 Out Subgraph_73 PrimaryInput))
