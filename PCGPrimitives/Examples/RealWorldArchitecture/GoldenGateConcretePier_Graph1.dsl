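; PCG example graph that builds a Golden Gate-style concrete bridge pier from primitive subgraphs.
; Single created points place two cylinders, a connecting block and three top blocks.
; East and west spline lines are offset and sampled, and one rotated side-cube point is copied onto both as side decoration.
; One spawner handles the cylinders and another all the block and decoration points, with both feeding the output.
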(graph GoldenGateConcretePier_Graph1
  (bind Cylinder1_Point (Subgraph :Graph /PCGPrimitives/Primitives/Create/Create_Point.Create_Point))
  (bind Cylinder2_Point (Subgraph :Graph /PCGPrimitives/Primitives/Create/Create_Point.Create_Point))
  (bind ConnectingBlock_Point (Subgraph :Graph /PCGPrimitives/Primitives/Create/Create_Point.Create_Point))
  (bind SpawnCylinders (Subgraph :Graph /PCGPrimitives/Primitives/Spawn/Spawn_Assets.Spawn_Assets))
  (bind EastSideLine (Subgraph :Graph /PCGPrimitives/Primitives/Create/Create_Spline_line.Create_Spline_line))
  (bind OffsetEastSide (Subgraph :Graph /PCGPrimitives/Primitives/Transform/Transform_Offset.Transform_Offset))
  (bind SampleEastSide (Subgraph :Graph /PCGPrimitives/Primitives/Place/Place_OnSpline.Place_OnSpline))
  (bind SideCube_Point (Subgraph :Graph /PCGPrimitives/Primitives/Create/Create_Point.Create_Point)) ; Those rotated cubes are used to create a decoration On the side of the pier. 
  (bind CopySideCubesToEast (Subgraph :Graph /PCGPrimitives/Primitives/Copy/CopyToPoints.CopyToPoints))
  (bind WestSideLine (Subgraph :Graph /PCGPrimitives/Primitives/Create/Create_Spline_line.Create_Spline_line))
  (bind OffsetWestSide (Subgraph :Graph /PCGPrimitives/Primitives/Transform/Transform_Offset.Transform_Offset))
  (bind SampleWestSide (Subgraph :Graph /PCGPrimitives/Primitives/Place/Place_OnSpline.Place_OnSpline))
  (bind CopySideCubesToWest (Subgraph :Graph /PCGPrimitives/Primitives/Copy/CopyToPoints.CopyToPoints))
  (bind MiddleTopBlock_Point (Subgraph :Graph /PCGPrimitives/Primitives/Create/Create_Point.Create_Point))
  (bind TopBlock1_Point (Subgraph :Graph /PCGPrimitives/Primitives/Create/Create_Point.Create_Point))
  (bind TopBlock2_Point (Subgraph :Graph /PCGPrimitives/Primitives/Create/Create_Point.Create_Point))
  (bind SpawnConnectingBlock (Subgraph :Graph /PCGPrimitives/Primitives/Spawn/Spawn_Assets.Spawn_Assets))
  (connect SideCube_Point Out CopySideCubesToEast Source)
  (connect SampleEastSide Out CopySideCubesToEast Target)
  (connect SideCube_Point Out CopySideCubesToWest Source)
  (connect SampleWestSide Out CopySideCubesToWest Target)
  (connect EastSideLine Out OffsetEastSide PrimaryInput)
  (connect WestSideLine Out OffsetWestSide PrimaryInput)
  (connect OffsetEastSide Out SampleEastSide PrimaryInput)
  (connect OffsetWestSide Out SampleWestSide PrimaryInput)
  (connect ConnectingBlock_Point Out SpawnConnectingBlock PrimaryInput)
  (connect CopySideCubesToEast Out SpawnConnectingBlock PrimaryInput)
  (connect CopySideCubesToWest Out SpawnConnectingBlock PrimaryInput)
  (connect MiddleTopBlock_Point Out SpawnConnectingBlock PrimaryInput)
  (connect TopBlock1_Point Out SpawnConnectingBlock PrimaryInput)
  (connect TopBlock2_Point Out SpawnConnectingBlock PrimaryInput)
  (connect Cylinder1_Point Out SpawnCylinders PrimaryInput)
  (connect Cylinder2_Point Out SpawnCylinders PrimaryInput)
  (output Out SpawnConnectingBlock)
  (output Out SpawnCylinders))
