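; PCG example graph that builds a Golden Gate-style bridge deck module from reusable PCGPrimitives subgraphs.
; Single points for upper and lower beams and left and right plates feed one asset spawner.
; A beam module fetched by actor reference is offset, then copied to points sampled along a lower spline line for vertical beams.
; These copies join two offset instances of a fetched truss module in a second spawner, and both spawners feed the output.
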(graph GoldenGateDeckModule_Graph
  (bind LowerBeam_Point (Subgraph :Graph /PCGPrimitives/Primitives/Create/Create_Point.Create_Point))
  (bind UpperBeam_Point (Subgraph :Graph /PCGPrimitives/Primitives/Create/Create_Point.Create_Point))
  (bind LowerBeamLine (Subgraph :Graph /PCGPrimitives/Primitives/Create/Create_Spline_line.Create_Spline_line))
  (bind SampleForVerticalBeams (Subgraph :Graph /PCGPrimitives/Primitives/Place/Place_OnSpline.Place_OnSpline))
  (bind GetBeamModule (Subgraph :Graph /PCGPrimitives/Primitives/Get/Get_ActorDataByRef.Get_ActorDataByRef))
  (bind Subgraph_4 (Subgraph :Graph /PCGPrimitives/Primitives/Transform/Transform_Offset.Transform_Offset))
  (bind CopyVerticalBeams (Subgraph :Graph /PCGPrimitives/Primitives/Copy/CopyToPoints.CopyToPoints))
  (bind LeftPlate (Subgraph :Graph /PCGPrimitives/Primitives/Create/Create_Point.Create_Point))
  (bind LeftPlateLower (Subgraph :Graph /PCGPrimitives/Primitives/Create/Create_Point.Create_Point))
  (bind LeftPlateLowerRight (Subgraph :Graph /PCGPrimitives/Primitives/Create/Create_Point.Create_Point))
  (bind RightPlate (Subgraph :Graph /PCGPrimitives/Primitives/Create/Create_Point.Create_Point))
  (bind RightPlateLower (Subgraph :Graph /PCGPrimitives/Primitives/Create/Create_Point.Create_Point))
  (bind RightPlateLowerRight (Subgraph :Graph /PCGPrimitives/Primitives/Create/Create_Point.Create_Point))
  (bind SpawnAllBeams (Subgraph :Graph /PCGPrimitives/Primitives/Spawn/Spawn_Assets.Spawn_Assets))
  (bind GetTrussModule (Subgraph :Graph /PCGPrimitives/Primitives/Get/Get_ActorDataByRef.Get_ActorDataByRef))
  (bind Subgraph_6 (Subgraph :Graph /PCGPrimitives/Primitives/Transform/Transform_Offset.Transform_Offset))
  (bind Subgraph_7 (Subgraph :Graph /PCGPrimitives/Primitives/Transform/Transform_Offset.Transform_Offset))
  (bind Subgraph_2 (Subgraph :Graph /PCGPrimitives/Primitives/Spawn/Spawn_Assets.Spawn_Assets))
  (connect Subgraph_4 Out CopyVerticalBeams Source)
  (connect SampleForVerticalBeams Out CopyVerticalBeams Target)
  (connect LowerBeamLine Out SampleForVerticalBeams PrimaryInput)
  (connect LeftPlate Out SpawnAllBeams PrimaryInput)
  (connect LeftPlateLower Out SpawnAllBeams PrimaryInput)
  (connect LeftPlateLowerRight Out SpawnAllBeams PrimaryInput)
  (connect LowerBeam_Point Out SpawnAllBeams PrimaryInput)
  (connect RightPlate Out SpawnAllBeams PrimaryInput)
  (connect RightPlateLower Out SpawnAllBeams PrimaryInput)
  (connect RightPlateLowerRight Out SpawnAllBeams PrimaryInput)
  (connect UpperBeam_Point Out SpawnAllBeams PrimaryInput)
  (connect CopyVerticalBeams Out Subgraph_2 PrimaryInput)
  (connect Subgraph_6 Out Subgraph_2 PrimaryInput)
  (connect Subgraph_7 Out Subgraph_2 PrimaryInput)
  (connect GetBeamModule Out Subgraph_4 PrimaryInput)
  (connect GetTrussModule Out Subgraph_6 PrimaryInput)
  (connect GetTrussModule Out Subgraph_7 PrimaryInput)
  (output Out SpawnAllBeams)
  (output Out Subgraph_2))
