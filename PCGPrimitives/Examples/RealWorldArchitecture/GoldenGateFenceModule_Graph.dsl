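; PCG example graph that builds a Golden Gate-style fence module from reusable PCGPrimitives subgraphs.
; A line spline feeds a sampler named for 40 cm spacing, alongside point creators for large posts, small posts and top rail.
; A single asset spawner receives the large post, top rail and small-post copy results, and provides the output.
; The small-post copy step has no wired inputs, leaving sampler and small-post point unconnected downstream.
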(graph GoldenGateFenceModule_Graph
  (bind FenceModuleLine (Subgraph :Graph /PCGPrimitives/Primitives/Create/Create_Spline_line.Create_Spline_line))
  (bind SampleEvery40cm (Subgraph :Graph /PCGPrimitives/Primitives/Place/Place_OnSpline.Place_OnSpline))
  (bind LargePost_Point (Subgraph :Graph /PCGPrimitives/Primitives/Create/Create_Point.Create_Point))
  (bind SmallPost_Point (Subgraph :Graph /PCGPrimitives/Primitives/Create/Create_Point.Create_Point))
  (bind CopySmallPostsToSampledPoints (Subgraph :Graph /PCGPrimitives/Primitives/Copy/CopyToPoints.CopyToPoints))
  (bind TopRail_Point (Subgraph :Graph /PCGPrimitives/Primitives/Create/Create_Point.Create_Point))
  (bind SpawnAllPosts (Subgraph :Graph /PCGPrimitives/Primitives/Spawn/Spawn_Assets.Spawn_Assets))
  (connect FenceModuleLine Out SampleEvery40cm PrimaryInput)
  (connect CopySmallPostsToSampledPoints Out SpawnAllPosts PrimaryInput)
  (connect LargePost_Point Out SpawnAllPosts PrimaryInput)
  (connect TopRail_Point Out SpawnAllPosts PrimaryInput)
  (output Out SpawnAllPosts))
