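; PCG example graph that builds an internal fence module for a Golden Gate architecture example, made of a post and repeated cable sections.
; One created cable-section point fans out through seven parallel offset primitives into a single asset spawner.
; A separate created point feeds the fence post spawner directly, and both spawners' results go to the graph's output.
; All steps are reusable PCGPrimitives Create, Transform and Spawn subgraphs.
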(graph GoldenGateInternalFenceModule_Graph
  (bind FencePost_Point (Subgraph :Graph /PCGPrimitives/Primitives/Create/Create_Point.Create_Point))
  (bind CableSection_Point (Subgraph :Graph /PCGPrimitives/Primitives/Create/Create_Point.Create_Point))
  (bind OffsetCable1 (Subgraph :Graph /PCGPrimitives/Primitives/Transform/Transform_Offset.Transform_Offset))
  (bind OffsetCable2 (Subgraph :Graph /PCGPrimitives/Primitives/Transform/Transform_Offset.Transform_Offset))
  (bind OffsetCable3 (Subgraph :Graph /PCGPrimitives/Primitives/Transform/Transform_Offset.Transform_Offset))
  (bind OffsetCable4 (Subgraph :Graph /PCGPrimitives/Primitives/Transform/Transform_Offset.Transform_Offset))
  (bind OffsetCable5 (Subgraph :Graph /PCGPrimitives/Primitives/Transform/Transform_Offset.Transform_Offset))
  (bind OffsetCable6 (Subgraph :Graph /PCGPrimitives/Primitives/Transform/Transform_Offset.Transform_Offset))
  (bind Subgraph_14 (Subgraph :Graph /PCGPrimitives/Primitives/Transform/Transform_Offset.Transform_Offset))
  (bind SpawnFenceModule (Subgraph :Graph /PCGPrimitives/Primitives/Spawn/Spawn_Assets.Spawn_Assets))
  (bind SpawnFencePost (Subgraph :Graph /PCGPrimitives/Primitives/Spawn/Spawn_Assets.Spawn_Assets))
  (connect CableSection_Point Out OffsetCable1 PrimaryInput)
  (connect CableSection_Point Out OffsetCable2 PrimaryInput)
  (connect CableSection_Point Out OffsetCable3 PrimaryInput)
  (connect CableSection_Point Out OffsetCable4 PrimaryInput)
  (connect CableSection_Point Out OffsetCable5 PrimaryInput)
  (connect CableSection_Point Out OffsetCable6 PrimaryInput)
  (connect OffsetCable1 Out SpawnFenceModule PrimaryInput)
  (connect OffsetCable2 Out SpawnFenceModule PrimaryInput)
  (connect OffsetCable3 Out SpawnFenceModule PrimaryInput)
  (connect OffsetCable4 Out SpawnFenceModule PrimaryInput)
  (connect OffsetCable5 Out SpawnFenceModule PrimaryInput)
  (connect OffsetCable6 Out SpawnFenceModule PrimaryInput)
  (connect Subgraph_14 Out SpawnFenceModule PrimaryInput)
  (connect FencePost_Point Out SpawnFencePost PrimaryInput)
  (connect CableSection_Point Out Subgraph_14 PrimaryInput)
  (output Out SpawnFenceModule)
  (output Out SpawnFencePost))
