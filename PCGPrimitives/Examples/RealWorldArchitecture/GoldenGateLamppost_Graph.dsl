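; PCG example graph that builds a Golden Gate lamppost from three parts: pole, arm and fixture.
; Each part comes from its own instance of the reusable Create_Point primitive subgraph.
; All three points merge into the primary input of one Spawn_Assets subgraph, whose result is the graph's output.
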(graph GoldenGateLamppost_Graph
  (bind LamppostPole_Point (Subgraph :Graph /PCGPrimitives/Primitives/Create/Create_Point.Create_Point))
  (bind LampArm_Point (Subgraph :Graph /PCGPrimitives/Primitives/Create/Create_Point.Create_Point))
  (bind LampFixture_Point (Subgraph :Graph /PCGPrimitives/Primitives/Create/Create_Point.Create_Point))
  (bind SpawnLamppost (Subgraph :Graph /PCGPrimitives/Primitives/Spawn/Spawn_Assets.Spawn_Assets))
  (connect LampArm_Point Out SpawnLamppost PrimaryInput)
  (connect LampFixture_Point Out SpawnLamppost PrimaryInput)
  (connect LamppostPole_Point Out SpawnLamppost PrimaryInput)
  (output Out SpawnLamppost))
