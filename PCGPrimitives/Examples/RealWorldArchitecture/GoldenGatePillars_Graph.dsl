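; PCG example graph that builds a Golden Gate-style bridge tower from PCGPrimitives subgraphs.
; Individual created points define the layered west leg, crossbeams, plates and X-braces, and a single offset transform duplicates the west leg to form the east one.
; Crossbeams 4–7 add decoration by offsetting a line spline east and west, sampling it and copying a cube point onto the samples.
; All elements feed one Spawn_Assets node, which is the graph output.
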(graph GoldenGatePillars_Graph
  (bind NT_WestLeg_Layer1 (Subgraph :Graph /PCGPrimitives/Primitives/Create/Create_Point.Create_Point))
  (bind NT_WestLeg_Layer2 (Subgraph :Graph /PCGPrimitives/Primitives/Create/Create_Point.Create_Point))
  (bind NT_WestLeg_Layer3 (Subgraph :Graph /PCGPrimitives/Primitives/Create/Create_Point.Create_Point))
  (bind NT_WestLeg_Layer4 (Subgraph :Graph /PCGPrimitives/Primitives/Create/Create_Point.Create_Point))
  (bind NT_Crossbeam1 (Subgraph :Graph /PCGPrimitives/Primitives/Create/Create_Point.Create_Point))
  (bind NT_Crossbeam2 (Subgraph :Graph /PCGPrimitives/Primitives/Create/Create_Point.Create_Point))
  (bind NT_Crossbeam4 (Subgraph :Graph /PCGPrimitives/Primitives/Create/Create_Point.Create_Point))
  (bind NT_Crossbeam5 (Subgraph :Graph /PCGPrimitives/Primitives/Create/Create_Point.Create_Point))
  (bind NT_Crossbeam6 (Subgraph :Graph /PCGPrimitives/Primitives/Create/Create_Point.Create_Point))
  (bind NT_Crossbeam7 (Subgraph :Graph /PCGPrimitives/Primitives/Create/Create_Point.Create_Point))
  (bind CB4_DecoCube (Subgraph :Graph /PCGPrimitives/Primitives/Create/Create_Point.Create_Point))
  (bind CB4_DecoLine (Subgraph :Graph /PCGPrimitives/Primitives/Create/Create_Spline_line.Create_Spline_line))
  (bind CB4_OffsetEast (Subgraph :Graph /PCGPrimitives/Primitives/Transform/Transform_Offset.Transform_Offset))
  (bind CB4_SampleEast (Subgraph :Graph /PCGPrimitives/Primitives/Place/Place_OnSpline.Place_OnSpline))
  (bind CB4_CopyToEast (Subgraph :Graph /PCGPrimitives/Primitives/Copy/CopyToPoints.CopyToPoints))
  (bind CB4_OffsetWest (Subgraph :Graph /PCGPrimitives/Primitives/Transform/Transform_Offset.Transform_Offset))
  (bind CB4_SampleWest (Subgraph :Graph /PCGPrimitives/Primitives/Place/Place_OnSpline.Place_OnSpline))
  (bind CB4_CopyToWest (Subgraph :Graph /PCGPrimitives/Primitives/Copy/CopyToPoints.CopyToPoints))
  (bind CB5_DecoCube (Subgraph :Graph /PCGPrimitives/Primitives/Create/Create_Point.Create_Point))
  (bind CB5_DecoLine (Subgraph :Graph /PCGPrimitives/Primitives/Create/Create_Spline_line.Create_Spline_line))
  (bind CB5_OffsetEast (Subgraph :Graph /PCGPrimitives/Primitives/Transform/Transform_Offset.Transform_Offset))
  (bind CB5_SampleEast (Subgraph :Graph /PCGPrimitives/Primitives/Place/Place_OnSpline.Place_OnSpline))
  (bind CB5_CopyToEast (Subgraph :Graph /PCGPrimitives/Primitives/Copy/CopyToPoints.CopyToPoints))
  (bind CB5_OffsetWest (Subgraph :Graph /PCGPrimitives/Primitives/Transform/Transform_Offset.Transform_Offset))
  (bind CB5_SampleWest (Subgraph :Graph /PCGPrimitives/Primitives/Place/Place_OnSpline.Place_OnSpline))
  (bind CB5_CopyToWest (Subgraph :Graph /PCGPrimitives/Primitives/Copy/CopyToPoints.CopyToPoints))
  (bind CB6_DecoCube (Subgraph :Graph /PCGPrimitives/Primitives/Create/Create_Point.Create_Point))
  (bind CB6_DecoLine (Subgraph :Graph /PCGPrimitives/Primitives/Create/Create_Spline_line.Create_Spline_line))
  (bind CB6_OffsetEast (Subgraph :Graph /PCGPrimitives/Primitives/Transform/Transform_Offset.Transform_Offset))
  (bind CB6_SampleEast (Subgraph :Graph /PCGPrimitives/Primitives/Place/Place_OnSpline.Place_OnSpline))
  (bind CB6_CopyToEast (Subgraph :Graph /PCGPrimitives/Primitives/Copy/CopyToPoints.CopyToPoints))
  (bind CB6_OffsetWest (Subgraph :Graph /PCGPrimitives/Primitives/Transform/Transform_Offset.Transform_Offset))
  (bind CB6_SampleWest (Subgraph :Graph /PCGPrimitives/Primitives/Place/Place_OnSpline.Place_OnSpline))
  (bind CB6_CopyToWest (Subgraph :Graph /PCGPrimitives/Primitives/Copy/CopyToPoints.CopyToPoints))
  (bind CB7_DecoCube (Subgraph :Graph /PCGPrimitives/Primitives/Create/Create_Point.Create_Point))
  (bind CB7_DecoLine (Subgraph :Graph /PCGPrimitives/Primitives/Create/Create_Spline_line.Create_Spline_line))
  (bind CB7_OffsetEast (Subgraph :Graph /PCGPrimitives/Primitives/Transform/Transform_Offset.Transform_Offset))
  (bind CB7_SampleEast (Subgraph :Graph /PCGPrimitives/Primitives/Place/Place_OnSpline.Place_OnSpline))
  (bind CB7_CopyToEast (Subgraph :Graph /PCGPrimitives/Primitives/Copy/CopyToPoints.CopyToPoints))
  (bind CB7_OffsetWest (Subgraph :Graph /PCGPrimitives/Primitives/Transform/Transform_Offset.Transform_Offset))
  (bind CB7_SampleWest (Subgraph :Graph /PCGPrimitives/Primitives/Place/Place_OnSpline.Place_OnSpline))
  (bind CB7_CopyToWest (Subgraph :Graph /PCGPrimitives/Primitives/Copy/CopyToPoints.CopyToPoints))
  (bind NT_WestLeg_Layer0 (Subgraph :Graph /PCGPrimitives/Primitives/Create/Create_Point.Create_Point))
  (bind NT_WestLeg_Layer0_Diagonal (Subgraph :Graph /PCGPrimitives/Primitives/Create/Create_Point.Create_Point))
  (bind NT_WestLeg_Layer0_Perpendicular (Subgraph :Graph /PCGPrimitives/Primitives/Create/Create_Point.Create_Point))
  (bind NT_WestLeg_Layer1_Diagonal (Subgraph :Graph /PCGPrimitives/Primitives/Create/Create_Point.Create_Point))
  (bind NT_WestLeg_Layer1_Perpendicular (Subgraph :Graph /PCGPrimitives/Primitives/Create/Create_Point.Create_Point))
  (bind NT_WestLeg_Layer2_Diagonal (Subgraph :Graph /PCGPrimitives/Primitives/Create/Create_Point.Create_Point))
  (bind NT_WestLeg_Layer2_Perpendicular (Subgraph :Graph /PCGPrimitives/Primitives/Create/Create_Point.Create_Point))
  (bind NT_WestLeg_Layer3_Diagonal (Subgraph :Graph /PCGPrimitives/Primitives/Create/Create_Point.Create_Point))
  (bind NT_WestLeg_Layer3_Perpendicular (Subgraph :Graph /PCGPrimitives/Primitives/Create/Create_Point.Create_Point))
  (bind DuplicateWestToEast (Subgraph :Graph /PCGPrimitives/Primitives/Transform/Transform_Offset.Transform_Offset))
  (bind NT_Crossbeam4_BottomPlate (Subgraph :Graph /PCGPrimitives/Primitives/Create/Create_Point.Create_Point))
  (bind NT_Crossbeam4_TopPlate (Subgraph :Graph /PCGPrimitives/Primitives/Create/Create_Point.Create_Point))
  (bind NT_Crossbeam5_BottomPlate (Subgraph :Graph /PCGPrimitives/Primitives/Create/Create_Point.Create_Point))
  (bind NT_Crossbeam5_TopPlate (Subgraph :Graph /PCGPrimitives/Primitives/Create/Create_Point.Create_Point))
  (bind NT_Crossbeam6_BottomPlate (Subgraph :Graph /PCGPrimitives/Primitives/Create/Create_Point.Create_Point))
  (bind NT_Crossbeam6_TopPlate (Subgraph :Graph /PCGPrimitives/Primitives/Create/Create_Point.Create_Point))
  (bind NT_Crossbeam7_BottomPlate (Subgraph :Graph /PCGPrimitives/Primitives/Create/Create_Point.Create_Point))
  (bind NT_Crossbeam7_TopPlate (Subgraph :Graph /PCGPrimitives/Primitives/Create/Create_Point.Create_Point))
  (bind NT_XBrace_CenterPlate (Subgraph :Graph /PCGPrimitives/Primitives/Create/Create_Point.Create_Point))
  (bind NT_XBrace_CornerPlate_BottomLeft (Subgraph :Graph /PCGPrimitives/Primitives/Create/Create_Point.Create_Point))
  (bind NT_XBrace_CornerPlate_BottomRight (Subgraph :Graph /PCGPrimitives/Primitives/Create/Create_Point.Create_Point))
  (bind NT_XBrace_CornerPlate_TopLeft (Subgraph :Graph /PCGPrimitives/Primitives/Create/Create_Point.Create_Point))
  (bind NT_XBrace_CornerPlate_TopRight (Subgraph :Graph /PCGPrimitives/Primitives/Create/Create_Point.Create_Point))
  (bind NT_XBrace_Lower_Diagonal1 (Subgraph :Graph /PCGPrimitives/Primitives/Create/Create_Point.Create_Point))
  (bind NT_XBrace_Lower_Diagonal2 (Subgraph :Graph /PCGPrimitives/Primitives/Create/Create_Point.Create_Point))
  (bind NT_XBrace_Upper_CenterPlate (Subgraph :Graph /PCGPrimitives/Primitives/Create/Create_Point.Create_Point))
  (bind NT_XBrace_Upper_Diagonal1 (Subgraph :Graph /PCGPrimitives/Primitives/Create/Create_Point.Create_Point))
  (bind NT_XBrace_Upper_Diagonal2 (Subgraph :Graph /PCGPrimitives/Primitives/Create/Create_Point.Create_Point))
  (bind SpawnAllTowerElements (Subgraph :Graph /PCGPrimitives/Primitives/Spawn/Spawn_Assets.Spawn_Assets))
  (connect CB4_DecoCube Out CB4_CopyToEast Source)
  (connect CB4_SampleEast Out CB4_CopyToEast Target)
  (connect CB4_DecoCube Out CB4_CopyToWest Source)
  (connect CB4_SampleWest Out CB4_CopyToWest Target)
  (connect CB4_DecoLine Out CB4_OffsetEast PrimaryInput)
  (connect CB4_DecoLine Out CB4_OffsetWest PrimaryInput)
  (connect CB4_OffsetEast Out CB4_SampleEast PrimaryInput)
  (connect CB4_OffsetWest Out CB4_SampleWest PrimaryInput)
  (connect CB5_DecoCube Out CB5_CopyToEast Source)
  (connect CB5_SampleEast Out CB5_CopyToEast Target)
  (connect CB5_DecoCube Out CB5_CopyToWest Source)
  (connect CB5_SampleWest Out CB5_CopyToWest Target)
  (connect CB5_DecoLine Out CB5_OffsetEast PrimaryInput)
  (connect CB5_DecoLine Out CB5_OffsetWest PrimaryInput)
  (connect CB5_OffsetEast Out CB5_SampleEast PrimaryInput)
  (connect CB5_OffsetWest Out CB5_SampleWest PrimaryInput)
  (connect CB6_DecoCube Out CB6_CopyToEast Source)
  (connect CB6_SampleEast Out CB6_CopyToEast Target)
  (connect CB6_DecoCube Out CB6_CopyToWest Source)
  (connect CB6_SampleWest Out CB6_CopyToWest Target)
  (connect CB6_DecoLine Out CB6_OffsetEast PrimaryInput)
  (connect CB6_DecoLine Out CB6_OffsetWest PrimaryInput)
  (connect CB6_OffsetEast Out CB6_SampleEast PrimaryInput)
  (connect CB6_OffsetWest Out CB6_SampleWest PrimaryInput)
  (connect CB7_DecoCube Out CB7_CopyToEast Source)
  (connect CB7_SampleEast Out CB7_CopyToEast Target)
  (connect CB7_DecoCube Out CB7_CopyToWest Source)
  (connect CB7_SampleWest Out CB7_CopyToWest Target)
  (connect CB7_DecoLine Out CB7_OffsetEast PrimaryInput)
  (connect CB7_DecoLine Out CB7_OffsetWest PrimaryInput)
  (connect CB7_OffsetEast Out CB7_SampleEast PrimaryInput)
  (connect CB7_OffsetWest Out CB7_SampleWest PrimaryInput)
  (connect NT_WestLeg_Layer0 Out DuplicateWestToEast PrimaryInput)
  (connect NT_WestLeg_Layer0_Diagonal Out DuplicateWestToEast PrimaryInput)
  (connect NT_WestLeg_Layer0_Perpendicular Out DuplicateWestToEast PrimaryInput)
  (connect NT_WestLeg_Layer1 Out DuplicateWestToEast PrimaryInput)
  (connect NT_WestLeg_Layer1_Diagonal Out DuplicateWestToEast PrimaryInput)
  (connect NT_WestLeg_Layer1_Perpendicular Out DuplicateWestToEast PrimaryInput)
  (connect NT_WestLeg_Layer2 Out DuplicateWestToEast PrimaryInput)
  (connect NT_WestLeg_Layer2_Diagonal Out DuplicateWestToEast PrimaryInput)
  (connect NT_WestLeg_Layer2_Perpendicular Out DuplicateWestToEast PrimaryInput)
  (connect NT_WestLeg_Layer3 Out DuplicateWestToEast PrimaryInput)
  (connect NT_WestLeg_Layer3_Diagonal Out DuplicateWestToEast PrimaryInput)
  (connect NT_WestLeg_Layer3_Perpendicular Out DuplicateWestToEast PrimaryInput)
  (connect NT_WestLeg_Layer4 Out DuplicateWestToEast PrimaryInput)
  (connect CB4_CopyToEast Out SpawnAllTowerElements PrimaryInput)
  (connect CB4_CopyToWest Out SpawnAllTowerElements PrimaryInput)
  (connect CB5_CopyToEast Out SpawnAllTowerElements PrimaryInput)
  (connect CB5_CopyToWest Out SpawnAllTowerElements PrimaryInput)
  (connect CB6_CopyToEast Out SpawnAllTowerElements PrimaryInput)
  (connect CB6_CopyToWest Out SpawnAllTowerElements PrimaryInput)
  (connect CB7_CopyToEast Out SpawnAllTowerElements PrimaryInput)
  (connect CB7_CopyToWest Out SpawnAllTowerElements PrimaryInput)
  (connect DuplicateWestToEast Out SpawnAllTowerElements PrimaryInput)
  (connect NT_Crossbeam1 Out SpawnAllTowerElements PrimaryInput)
  (connect NT_Crossbeam2 Out SpawnAllTowerElements PrimaryInput)
  (connect NT_Crossbeam4 Out SpawnAllTowerElements PrimaryInput)
  (connect NT_Crossbeam4_BottomPlate Out SpawnAllTowerElements PrimaryInput)
  (connect NT_Crossbeam4_TopPlate Out SpawnAllTowerElements PrimaryInput)
  (connect NT_Crossbeam5 Out SpawnAllTowerElements PrimaryInput)
  (connect NT_Crossbeam5_BottomPlate Out SpawnAllTowerElements PrimaryInput)
  (connect NT_Crossbeam5_TopPlate Out SpawnAllTowerElements PrimaryInput)
  (connect NT_Crossbeam6 Out SpawnAllTowerElements PrimaryInput)
  (connect NT_Crossbeam6_BottomPlate Out SpawnAllTowerElements PrimaryInput)
  (connect NT_Crossbeam6_TopPlate Out SpawnAllTowerElements PrimaryInput)
  (connect NT_Crossbeam7 Out SpawnAllTowerElements PrimaryInput)
  (connect NT_Crossbeam7_BottomPlate Out SpawnAllTowerElements PrimaryInput)
  (connect NT_Crossbeam7_TopPlate Out SpawnAllTowerElements PrimaryInput)
  (connect NT_WestLeg_Layer0 Out SpawnAllTowerElements PrimaryInput)
  (connect NT_WestLeg_Layer0_Diagonal Out SpawnAllTowerElements PrimaryInput)
  (connect NT_WestLeg_Layer0_Perpendicular Out SpawnAllTowerElements PrimaryInput)
  (connect NT_WestLeg_Layer1 Out SpawnAllTowerElements PrimaryInput)
  (connect NT_WestLeg_Layer1_Diagonal Out SpawnAllTowerElements PrimaryInput)
  (connect NT_WestLeg_Layer1_Perpendicular Out SpawnAllTowerElements PrimaryInput)
  (connect NT_WestLeg_Layer2 Out SpawnAllTowerElements PrimaryInput)
  (connect NT_WestLeg_Layer2_Diagonal Out SpawnAllTowerElements PrimaryInput)
  (connect NT_WestLeg_Layer2_Perpendicular Out SpawnAllTowerElements PrimaryInput)
  (connect NT_WestLeg_Layer3 Out SpawnAllTowerElements PrimaryInput)
  (connect NT_WestLeg_Layer3_Diagonal Out SpawnAllTowerElements PrimaryInput)
  (connect NT_WestLeg_Layer3_Perpendicular Out SpawnAllTowerElements PrimaryInput)
  (connect NT_WestLeg_Layer4 Out SpawnAllTowerElements PrimaryInput)
  (connect NT_XBrace_CenterPlate Out SpawnAllTowerElements PrimaryInput)
  (connect NT_XBrace_CornerPlate_BottomLeft Out SpawnAllTowerElements PrimaryInput)
  (connect NT_XBrace_CornerPlate_BottomRight Out SpawnAllTowerElements PrimaryInput)
  (connect NT_XBrace_CornerPlate_TopLeft Out SpawnAllTowerElements PrimaryInput)
  (connect NT_XBrace_CornerPlate_TopRight Out SpawnAllTowerElements PrimaryInput)
  (connect NT_XBrace_Lower_Diagonal1 Out SpawnAllTowerElements PrimaryInput)
  (connect NT_XBrace_Lower_Diagonal2 Out SpawnAllTowerElements PrimaryInput)
  (connect NT_XBrace_Upper_CenterPlate Out SpawnAllTowerElements PrimaryInput)
  (connect NT_XBrace_Upper_Diagonal1 Out SpawnAllTowerElements PrimaryInput)
  (connect NT_XBrace_Upper_Diagonal2 Out SpawnAllTowerElements PrimaryInput)
  (output Out SpawnAllTowerElements))
