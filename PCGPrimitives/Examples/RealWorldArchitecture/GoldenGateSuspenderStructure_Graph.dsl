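; PCG example graph that procedurally builds the suspender structure of a Golden Gate style bridge from PCGPrimitives subgraphs.
; Authored tower, anchor, intermediate and mid-road points connect into center, north and south cable paths, offset east and west and sampled for spawned sections.
; Resampled cable points bridge to offset copies, with those polylines rebuilt as linear splines for spawning connecting pieces.
; A third branch spawns assets at offset resampled points.
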(graph GoldenGateSuspenderStructure_Graph
  (bind NorthTowerTop_Point (Subgraph :Graph /PCGPrimitives/Primitives/Create/Create_Point.Create_Point))
  (bind SouthTowerTop_Point (Subgraph :Graph /PCGPrimitives/Primitives/Create/Create_Point.Create_Point))
  (bind MiddleRoadTop_Point (Subgraph :Graph /PCGPrimitives/Primitives/Create/Create_Point.Create_Point))
  (bind NorthIntermediate_Point (Subgraph :Graph /PCGPrimitives/Primitives/Create/Create_Point.Create_Point))
  (bind SouthIntermediate_Point (Subgraph :Graph /PCGPrimitives/Primitives/Create/Create_Point.Create_Point))
  (bind ConnectSuspenderPoints (Subgraph :Graph /PCGPrimitives/Primitives/Trace/Trace_Connect.Trace_Connect))
  (bind OffsetEastSuspender (Subgraph :Graph /PCGPrimitives/Primitives/Transform/Transform_Offset.Transform_Offset))
  (bind OffsetWestSuspender (Subgraph :Graph /PCGPrimitives/Primitives/Transform/Transform_Offset.Transform_Offset))
  (bind NorthAnchor_Point (Subgraph :Graph /PCGPrimitives/Primitives/Create/Create_Point.Create_Point))
  (bind NorthIntermediateEnd_Point (Subgraph :Graph /PCGPrimitives/Primitives/Create/Create_Point.Create_Point))
  (bind ConnectNorthSuspender (Subgraph :Graph /PCGPrimitives/Primitives/Trace/Trace_Connect.Trace_Connect))
  (bind OffsetNorthEastSuspender (Subgraph :Graph /PCGPrimitives/Primitives/Transform/Transform_Offset.Transform_Offset))
  (bind OffsetNorthWestSuspender (Subgraph :Graph /PCGPrimitives/Primitives/Transform/Transform_Offset.Transform_Offset))
  (bind SouthAnchor_Point (Subgraph :Graph /PCGPrimitives/Primitives/Create/Create_Point.Create_Point))
  (bind SouthIntermediateEnd_Point (Subgraph :Graph /PCGPrimitives/Primitives/Create/Create_Point.Create_Point))
  (bind ConnectSouthSuspender (Subgraph :Graph /PCGPrimitives/Primitives/Trace/Trace_Connect.Trace_Connect))
  (bind OffsetSouthEastSuspender (Subgraph :Graph /PCGPrimitives/Primitives/Transform/Transform_Offset.Transform_Offset))
  (bind OffsetSouthWestSuspender (Subgraph :Graph /PCGPrimitives/Primitives/Transform/Transform_Offset.Transform_Offset))
  (bind SampleSuspenderEvery8m (Subgraph :Graph /PCGPrimitives/Primitives/Place/Place_OnSpline.Place_OnSpline))
  (bind SuspenderSection_Point (Subgraph :Graph /PCGPrimitives/Primitives/Create/Create_Point.Create_Point))
  (bind CopySuspenderSections (Subgraph :Graph /PCGPrimitives/Primitives/Copy/CopyToPoints.CopyToPoints))
  (bind SpawnSuspenderSections (Subgraph :Graph /PCGPrimitives/Primitives/Spawn/Spawn_Assets.Spawn_Assets))
  (bind Subgraph_0 (Subgraph :Graph /PCGPrimitives/Primitives/Transform/Transform_Resample.Transform_Resample))
  (bind Subgraph_1 (Subgraph :Graph /PCGPrimitives/Primitives/Transform/Transform_Offset.Transform_Offset))
  (bind Subgraph_2 (Subgraph :Graph /PCGPrimitives/Primitives/Trace/Trace_Bridges.Trace_Bridges))
  (bind Subgraph_3 (Subgraph :Graph /PCGPrimitives/Primitives/Transform/Transform_Resample.Transform_Resample))
  (bind Subgraph_4 (Subgraph :Graph /PCGPrimitives/Primitives/Transform/Transform_Offset.Transform_Offset))
  (bind Subgraph_5 (Subgraph :Graph /PCGPrimitives/Primitives/Trace/Trace_Bridges.Trace_Bridges))
  (bind Subgraph_6 (Subgraph :Graph /PCGPrimitives/Primitives/Create/Create_Point.Create_Point))
  (bind Subgraph_10 (Subgraph :Graph /PCGPrimitives/Primitives/Transform/Transform_Resample.Transform_Resample))
  (bind Subgraph_11 (Subgraph :Graph /PCGPrimitives/Primitives/Transform/Transform_Offset.Transform_Offset))
  (bind Subgraph_12 (Subgraph :Graph /PCGPrimitives/Primitives/Trace/Trace_Bridges.Trace_Bridges))
  (bind Subgraph_7 (Subgraph :Graph /PCGPrimitives/Primitives/Place/Place_OnSpline.Place_OnSpline))
  (bind Subgraph_9 (Subgraph :Graph /PCGPrimitives/Primitives/Copy/CopyToPoints.CopyToPoints))
  (bind Subgraph_8 (Subgraph :Graph /PCGPrimitives/Primitives/Spawn/Spawn_Assets.Spawn_Assets))
  (bind Subgraph_13 (Subgraph :Graph /PCGPrimitives/Primitives/Transform/Transform_Offset.Transform_Offset))
  (bind Subgraph_14 (Subgraph :Graph /PCGPrimitives/Primitives/Create/Create_Point.Create_Point))
  (bind Subgraph_15 (Subgraph :Graph /PCGPrimitives/Primitives/Copy/CopyToPoints.CopyToPoints))
  (bind Subgraph_16 (Subgraph :Graph /PCGPrimitives/Primitives/Spawn/Spawn_Assets.Spawn_Assets))
  ;[0.15 0.15 0.15 0.5] Forcing_spline_linear
  (bind FilterDataByType_0 (Filter%20Data%20By%20Type :TargetType (struct :Ids [(struct :Struct /Script/PCG.PCGDataTypeInfoPolyline)])))
  (bind GetSplineControlPoints_0 (Get%20Spline%20Control%20Points))
  (bind CreateSpline_0 (Create%20Spline :bLinear true))
  (connect NorthAnchor_Point Out ConnectNorthSuspender PrimaryInput)
  (connect NorthIntermediateEnd_Point Out ConnectNorthSuspender PrimaryInput)
  (connect NorthTowerTop_Point Out ConnectNorthSuspender PrimaryInput)
  (connect SouthAnchor_Point Out ConnectSouthSuspender PrimaryInput)
  (connect SouthIntermediateEnd_Point Out ConnectSouthSuspender PrimaryInput)
  (connect SouthTowerTop_Point Out ConnectSouthSuspender PrimaryInput)
  (connect MiddleRoadTop_Point Out ConnectSuspenderPoints PrimaryInput)
  (connect NorthIntermediate_Point Out ConnectSuspenderPoints PrimaryInput)
  (connect NorthTowerTop_Point Out ConnectSuspenderPoints PrimaryInput)
  (connect SouthIntermediate_Point Out ConnectSuspenderPoints PrimaryInput)
  (connect SouthTowerTop_Point Out ConnectSuspenderPoints PrimaryInput)
  (connect SuspenderSection_Point Out CopySuspenderSections Source)
  (connect SampleSuspenderEvery8m Out CopySuspenderSections Target)
  (connect GetSplineControlPoints_0 Out CreateSpline_0 In)
  (connect Subgraph_2 Out FilterDataByType_0 In)
  (connect Subgraph_5 Out FilterDataByType_0 In)
  (connect Subgraph_12 Out FilterDataByType_0 In)
  (connect FilterDataByType_0 InsideFilter GetSplineControlPoints_0 In)
  (connect ConnectSuspenderPoints Out OffsetEastSuspender PrimaryInput)
  (connect ConnectNorthSuspender Out OffsetNorthEastSuspender PrimaryInput)
  (connect ConnectNorthSuspender Out OffsetNorthWestSuspender PrimaryInput)
  (connect ConnectSouthSuspender Out OffsetSouthEastSuspender PrimaryInput)
  (connect ConnectSouthSuspender Out OffsetSouthWestSuspender PrimaryInput)
  (connect ConnectSuspenderPoints Out OffsetWestSuspender PrimaryInput)
  (connect OffsetEastSuspender Out SampleSuspenderEvery8m PrimaryInput)
  (connect OffsetNorthEastSuspender Out SampleSuspenderEvery8m PrimaryInput)
  (connect OffsetNorthWestSuspender Out SampleSuspenderEvery8m PrimaryInput)
  (connect OffsetSouthEastSuspender Out SampleSuspenderEvery8m PrimaryInput)
  (connect OffsetSouthWestSuspender Out SampleSuspenderEvery8m PrimaryInput)
  (connect OffsetWestSuspender Out SampleSuspenderEvery8m PrimaryInput)
  (connect CopySuspenderSections Out SpawnSuspenderSections PrimaryInput)
  (connect OffsetEastSuspender Out Subgraph_0 PrimaryInput)
  (connect Subgraph_0 Out Subgraph_1 PrimaryInput)
  (connect Subgraph_0 Out Subgraph_2 PrimaryInput)
  (connect Subgraph_1 Out Subgraph_2 SecondaryInput)
  (connect OffsetWestSuspender Out Subgraph_3 PrimaryInput)
  (connect Subgraph_3 Out Subgraph_4 PrimaryInput)
  (connect Subgraph_3 Out Subgraph_5 PrimaryInput)
  (connect Subgraph_4 Out Subgraph_5 SecondaryInput)
  (connect CreateSpline_0 Out Subgraph_7 PrimaryInput)
  (connect Subgraph_9 Out Subgraph_8 PrimaryInput)
  (connect Subgraph_6 Out Subgraph_9 Source)
  (connect Subgraph_7 Out Subgraph_9 Target)
  (connect OffsetNorthEastSuspender Out Subgraph_10 PrimaryInput)
  (connect OffsetNorthWestSuspender Out Subgraph_10 PrimaryInput)
  (connect OffsetSouthEastSuspender Out Subgraph_10 PrimaryInput)
  (connect OffsetSouthWestSuspender Out Subgraph_10 PrimaryInput)
  (connect Subgraph_10 Out Subgraph_11 PrimaryInput)
  (connect Subgraph_10 Out Subgraph_12 PrimaryInput)
  (connect Subgraph_11 Out Subgraph_12 SecondaryInput)
  (connect Subgraph_0 Out Subgraph_13 PrimaryInput)
  (connect Subgraph_3 Out Subgraph_13 PrimaryInput)
  (connect Subgraph_10 Out Subgraph_13 PrimaryInput)
  (connect Subgraph_14 Out Subgraph_15 Source)
  (connect Subgraph_13 Out Subgraph_15 Target)
  (connect Subgraph_15 Out Subgraph_16 PrimaryInput))
