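; PCG example graph that builds a Golden Gate-style truss module entirely from PCGPrimitives subgraphs.
; A center spline, one meter shorter than the main beam to allow for the X-braces, is offset left and right, then both lines are sampled.
; Up and down diagonal brace points are copied onto those samples to form X-braces.
; These braces, top and bottom beam points, and four corner end-plate points feed one asset spawner, which is the graph's output.
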(graph GoldenGateTrussModule_Graph
  (bind CenterTrussLine (Subgraph :Graph /PCGPrimitives/Primitives/Create/Create_Spline_line.Create_Spline_line)) ; Notice how the center line is one meter shorter than \nthe main beam to account for the X-braces spawning. 
  (bind OffsetLineLeft (Subgraph :Graph /PCGPrimitives/Primitives/Transform/Transform_Offset.Transform_Offset))
  (bind OffsetLineRight (Subgraph :Graph /PCGPrimitives/Primitives/Transform/Transform_Offset.Transform_Offset))
  (bind SampleBothLines (Subgraph :Graph /PCGPrimitives/Primitives/Place/Place_OnSpline.Place_OnSpline))
  (bind SpawnTrussModule (Subgraph :Graph /PCGPrimitives/Primitives/Spawn/Spawn_Assets.Spawn_Assets))
  ;[0.15 0.15 0.15 0.5] TOP and BOTTOM Plates
  (bind TopBeam_Point (Subgraph :Graph /PCGPrimitives/Primitives/Create/Create_Point.Create_Point))
  (bind BottomBeam_Point (Subgraph :Graph /PCGPrimitives/Primitives/Create/Create_Point.Create_Point))
  ;[0.15 0.15 0.15 0.5] X Braces
  (bind DiagonalBraceUp (Subgraph :Graph /PCGPrimitives/Primitives/Create/Create_Point.Create_Point))
  (bind DiagonalBraceDown (Subgraph :Graph /PCGPrimitives/Primitives/Create/Create_Point.Create_Point))
  (bind CopyUpDiagonalsToSamples (Subgraph :Graph /PCGPrimitives/Primitives/Copy/CopyToPoints.CopyToPoints))
  (bind CopyDownDiagonalsToSamples (Subgraph :Graph /PCGPrimitives/Primitives/Copy/CopyToPoints.CopyToPoints))
  ;[0.15 0.15 0.15 0.5] PLATES
  (bind EndPlateBottomLeft (Subgraph :Graph /PCGPrimitives/Primitives/Create/Create_Point.Create_Point))
  (bind EndPlateBottomRight (Subgraph :Graph /PCGPrimitives/Primitives/Create/Create_Point.Create_Point))
  (bind EndPlateTopLeft (Subgraph :Graph /PCGPrimitives/Primitives/Create/Create_Point.Create_Point))
  (bind EndPlateTopRight (Subgraph :Graph /PCGPrimitives/Primitives/Create/Create_Point.Create_Point))
  (connect DiagonalBraceDown Out CopyDownDiagonalsToSamples Source)
  (connect SampleBothLines Out CopyDownDiagonalsToSamples Target)
  (connect DiagonalBraceUp Out CopyUpDiagonalsToSamples Source)
  (connect SampleBothLines Out CopyUpDiagonalsToSamples Target)
  (connect CenterTrussLine Out OffsetLineLeft PrimaryInput)
  (connect CenterTrussLine Out OffsetLineRight PrimaryInput)
  (connect OffsetLineLeft Out SampleBothLines PrimaryInput)
  (connect OffsetLineRight Out SampleBothLines PrimaryInput)
  (connect BottomBeam_Point Out SpawnTrussModule PrimaryInput)
  (connect CopyDownDiagonalsToSamples Out SpawnTrussModule PrimaryInput)
  (connect CopyUpDiagonalsToSamples Out SpawnTrussModule PrimaryInput)
  (connect EndPlateBottomLeft Out SpawnTrussModule PrimaryInput)
  (connect EndPlateBottomRight Out SpawnTrussModule PrimaryInput)
  (connect EndPlateTopLeft Out SpawnTrussModule PrimaryInput)
  (connect EndPlateTopRight Out SpawnTrussModule PrimaryInput)
  (connect TopBeam_Point Out SpawnTrussModule PrimaryInput)
  (output Out SpawnTrussModule))
